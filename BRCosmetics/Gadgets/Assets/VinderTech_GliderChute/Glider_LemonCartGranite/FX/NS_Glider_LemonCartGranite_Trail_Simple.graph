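# Niagara trail effect for the LemonCartGranite glider, layering flickering mesh flames, light-emitting flares, glows, flower bursts, mesh ripples and speed lines.
# Particle partitions or spawn groups pin the jet layers to two points on the glider's skeletal mesh.
# Smoothed ForwardDot scales size and brightness, while flame materials receive a lagged position.
# Flames also follow time-of-day color and reactivity timing, and front-end modifiers separate lobby from gameplay tuning.
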NiagaraSystem /BRCosmetics/Gadgets/Assets/VinderTech_GliderChute/Glider_LemonCartGranite/FX/NS_Glider_LemonCartGranite_Trail_Simple
dependencies {
 usedDataInterfaces [
  &NiagaraDataInterfaceArrayInt32
  &NiagaraDataInterfaceArrayFloat3
  &NiagaraDataInterfaceCurve
  &NiagaraDataInterfaceArrayFloat
  &NiagaraDataInterfaceVector2DCurve
  &NiagaraDataInterfaceColorCurve
  &NiagaraDataInterfaceSkeletalMesh
 ]
 usedDynamicInputs [
  &SimulationPosition
  &Abs_Float
  &SetBoolByFloatComparison
 ]
 usedModules [
  &SystemState
  &self:SystemSpawnScript.SystemScriptSource.SystemScriptGraph.NiagaraNodeAssignment_2.SetVariables_#1
  &self:SystemSpawnScript.SystemScriptSource.SystemScriptGraph.NiagaraNodeAssignment_0.SetVariables_#2
  &self:SystemSpawnScript.SystemScriptSource.SystemScriptGraph.NiagaraNodeAssignment_1.SetVariables_#3
  &Timeline
  &self:SystemSpawnScript.SystemScriptSource.SystemScriptGraph.NiagaraNodeAssignment_3.SetVariables_#4
  &self:SystemSpawnScript.SystemScriptSource.SystemScriptGraph.NiagaraNodeAssignment_4.SetVariables_#5
 ]
}
emitters {
 Flame {
  emitterUpdateScript [
   &EmitterState(
    Inactive Response=::NewEnumerator1
    Life Cycle Mode=::NewEnumerator1
    Loop Behavior=::NewEnumerator1
    Loop Delay=0.2
    Loop Duration Mode=::NewEnumerator1
    UseLoopDelay=true
   )
   &SpawnBurst_Instantaneous(Spawn Count=2, Use Spawn Probability=false)
   /Niagara/Noise/Waveform(
    EnableExponent1=false
    GlobalClampMax=false
    GlobalClampMin=false
    UseInput1=false
    UsePhase1=false
    "Waveform [1]":ENiagara_Waveforms=::NewEnumerator12
    "[1] Frequency"=7
   )
   /SetVariables(Emitter.Flicker:NiagaraFloat=&Add_Float(A=&Multiply_Float(A=$Emitter.Waveform.Waveform, B=0.5), B=0.5)) {isSetParameters=true, module=SetVariables_#6}
   /SetVariables(Emitter.bInterpolation=true) {isSetParameters=true, module=SetVariables_#7}
  ]
  particleSpawnScript [
   &InitializeParticle(
    Mass Mode=::NewEnumerator1
    Mesh Scale=&Multiply_Vector(A=&MakeVector(X=3.4, Y=2, Z=2), B=&VectorFromFloat(Value=0.4))
    Mesh Scale Mode=::NewEnumerator4
    Position Mode=::NewEnumerator0
    Sprite Size Mode=::NewEnumerator3
    Uniform Sprite Size=8
    Write Material Random=true
   ) {module=InitializeParticle001}
   &PartitionParticles()
   &CameraOffset()
   /SetVariables() {isSetParameters=true, module=SetVariables_#8}
   &SkeletalMeshLocation(
    Apply Sampled Orientation to Mesh Particles=true
    Bone / Socket Index=&SelectIntFromArray(Direct Array Index=$Particles.PartitionParticles.Partition, Int Selection Array=DataInterface() {propertyValues {IntData [0,1]}})
    Inherit Velocity=false
    Interpolate Mesh Sampling=false
    MaskSpawnGroup=false
    Mesh Sampling Type=::NewEnumerator5
    Sampled Position Offset Coordinate Space:ENiagaraMeshTransforms=::NewEnumerator3
    Skeletal Mesh=$System.Skeletal Mesh
    Socket Sampling Mode=::NewEnumerator5
    Use Sampled Mesh Attribute as Ribbon ID=false
   ) {module=SkeletalMeshLocation001}
   &FN_Sprite_Attributes(
    Color=&LerpLinearColors(
     StartColor=<rgba 1 0.9047 0.2831 1>
     EndColor=<rgba 0.9823 0.9131 0.2582 1>
     LerpFactor=&FN_FrontEndFloatModifier(Gameplay=0, Front End=1)
    )
    Create Color Attribute=true
    Create SpriteSize Attribute=true
    SpriteSize=<20 20>
    Write to Module Namespace=true
   ) {module=FN_Sprite_Attributes001}
   &InitialMeshOrientation(
    Mesh Axis To Orient=&SelectVectorFromArray(
     Direct Array Index=$Particles.PartitionParticles.Partition
     Vector Selection Array=DataInterface() {propertyValues {FloatData [<1 0 0>, <-1 0 0>]}}
    )
    Mesh Orientation Mode=::NewEnumerator2
    Orientation Vector=&RotateVectorByQuaternion(Vector To Rotate=<0 1 0>, Quaternion=$Output.SkeletalMeshLocation001.SampledRotation)
    Use Rotation Vector=false
   )
  ]
  particleUpdateScript [
   &ParticleState(Kill Particles When Lifetime Has Elapsed=false, Let Infinitely Lived Particles Die When Emitter Deactivates=true)
   &SkeletalMeshLocation(
    Apply Sampled Orientation to Mesh Particles=true
    Bone / Socket Index=&SelectIntFromArray(Direct Array Index=$Particles.PartitionParticles.Partition, Int Selection Array=DataInterface() {propertyValues {IntData [0,1]}})
    Inherit Velocity=false
    Interpolate Mesh Sampling=false
    MaskSpawnGroup=false
    Mesh Sampling Type=::NewEnumerator5
    Sampled Position Offset=&SelectVectorFromArray(
     Direct Array Index=$Particles.Initial.PartitionParticles.Partition
     Vector Selection Array=DataInterface() {propertyValues {FloatData [<-2.5 0 0>, <2.5 0 0>]}}
    )
    Sampled Position Offset Coordinate Space:ENiagaraMeshTransforms=::NewEnumerator3
    Skeletal Mesh=$System.Skeletal Mesh
    Socket Sampling Mode=::NewEnumerator5
    Use Sampled Mesh Attribute as Ribbon ID=false
   )
   /SetVariables(Particles.MeshOrientation=&MultiplyQuaternion(
     Quaternion A=$Output.SkeletalMeshLocation.SampledRotation
     Quaternion B=&MakeQuatFromAxisAngle(
      Axis=&SelectVectorFromArray(
       Direct Array Index=$Particles.PartitionParticles.Partition
       Vector Selection Array=DataInterface() {propertyValues {FloatData [<-1 0 0>, <-1 0 0>]}}
      )
      AngleInDegrees=0
     )
    )) {enabled=false, isSetParameters=true}
   /SetVariables(Particles.MeshOrientation=&MultiplyQuaternion(
     Quaternion A=$Particles.SkeletalMeshLocation.SampledRotation
     Quaternion B=&MakeQuatFromAxisAngle(
      Axis=&SelectVectorFromArray(
       Direct Array Index=$Particles.PartitionParticles.Partition
       Vector Selection Array=DataInterface() {propertyValues {FloatData [<-1 0 0>, <-1 0 0>]}}
      )
      AngleInDegrees=0
     )
    )) {enabled=false, isSetParameters=true}
   /SetVariables(Particles.Velocity=&RotateVectorByQuaternion(Vector To Rotate=<0 0 1>, Quaternion=$Output.SkeletalMeshLocation.SampledRotation)) {isSetParameters=true, module=SetVariables_#9}
   &InterpolateOverTime(
    Rate Of Change=8
    Target Vector Value:Vector3f=&ConvertPositionToVector(Input Position=$Particles.Position, Convert To Absolute Large World Vector=false)
    Variable Type=::NewEnumerator2
   ) {enabled=false}
   /SetVariables(Particles.Interpolate Dynamic=&MakeFloat4FromVectorAndFloat(
     "Vector (RGB)"=$StackContext.InterpolateOverTime.InterpolateOverTime_V3_.Moving Average
     "Float (Alpha)"=&Add_Float(A=$User.Interpolate Intensity, B=-0.5)
    )) {isSetParameters=true}
   &FN_Sprite_Attributes(
    Camera Offset=5
    Create Camera Offset Attribute=true
    Create Color Attribute=true
    Create Position Attribute=true
    Position:NiagaraPosition=&AddVectorToPosition(
     Position=$Output.SkeletalMeshLocation.SampledPosition
     Vector=&TransformVector(Vector=<-0.8 0 2.5>, Destination Space=::World)
    )
    Write to Module Namespace=true
   )
   &ScaleColor(Scale Alpha=&Multiply_Float(A=$Emitter.Flicker, B=4), ScaleA=true, ScaleRGB=true)
   &InterpolateOverTime(Rate Of Change=4, Target Float Value:NiagaraFloat=$User.ForwardDot) {module=InterpolateOverTime001}
   &ScaleMeshSize(Scale Factor=&MakeVector(
     X=&FloatFromCurve(
      FloatCurve=DataInterface() {propertyValues {Curve {keys [
          {interpMode=RCIM_Cubic, tangentMode=RCTM_User, time=0, value=0}
          {interpMode=RCIM_Cubic, tangentMode=RCTM_User, time=0.3, value=1}
          {tangentMode=RCTM_User, time=1, value=1}
         ]}}}
      Scale Curve=&Lerp_Float(A=1, B=1.25, Alpha=$Particles.InterpolateOverTime001.InterpolateOverTime_Float.Moving Average)
     )
     Y=0.5
     Z=&FloatFromCurve(
      FloatCurve=DataInterface() {propertyValues {Curve {keys [
          {arriveTangent=15.3111, interpMode=RCIM_Cubic, leaveTangent=15.3111, tangentMode=RCTM_User, time=0, value=0}
          {arriveTangent=-0.0274, interpMode=RCIM_Cubic, leaveTangent=-0.0274, tangentMode=RCTM_User, time=0.2, value=1.375}
          {tangentMode=RCTM_User, time=1, value=1}
         ]}}}
      Scale Curve=&Lerp_Float(A=1.5, B=2.5, Alpha=$Particles.InterpolateOverTime001.InterpolateOverTime_Float.Moving Average)
     )
    )) {enabled=false, module=ScaleMeshSize001}
   &ScaleColor(
    Scale Mode=::NewEnumerator1
    Scale RGBA=&MakeVector4_FromLinearColor(LinearColor=$System.TODColor)
    ScaleRGBA=true
   ) {module=ScaleColor001}
   &ScaleColor(
    Scale Alpha=&FloatFromCurve(
     FloatCurve=DataInterface() {propertyValues {Curve {keys [<at 0.2 1 RCIM_Cubic>,<at 0.4 1 RCIM_Cubic>]}}}
     CurveIndex=$System.ReactivityTiming
    )
    Scale RGB=&VectorFromFloat(Value=&FN_FrontEndFloatModifier(
      Gameplay=&FloatFromCurve(
       FloatCurve=DataInterface() {propertyValues {Curve {keys [<at 0.2 1 RCIM_Cubic>, {interpMode=RCIM_Cubic, tangentMode=RCTM_User, time=0.25, value=5}, <at 0.4 1 RCIM_Cubic>]}}}
       CurveIndex=$System.ReactivityTiming
      )
      Front End=&FloatFromCurve(
       FloatCurve=DataInterface() {propertyValues {Curve {keys [<at 0.2 1 RCIM_Cubic>, {interpMode=RCIM_Cubic, tangentMode=RCTM_User, time=0.25, value=3}, <at 0.4 1 RCIM_Cubic>]}}}
       CurveIndex=$System.ReactivityTiming
      )
     ))
    ScaleA=true
    ScaleRGB=false
   ) {module=ScaleColor002}
   /SetVariables(Particles.GlowColor=&MultiplyLinearColor(
     Linear Color=$Particles.Color
     Scale Factor=&MakeVector4_FromLinearColor(LinearColor=<rgba 1 0.8963 0.2747 1>)
    )) {isSetParameters=true, module=SetVariables_#10}
   &InterpolateOverTime(Rate Of Change=3, Target Float Value:NiagaraFloat=$User.RightDot) {module=InterpolateOverTime002}
   &InterpolateOverTime(Rate Of Change=20, Target Vector Value:NiagaraPosition=$Particles.Position, Variable Type=::NewEnumerator2) {module=InterpolateOverTime003}
   &DynamicMaterialParameters(
    Index 0 Param 1=&Lerp_Float(A=3.5, B=5, Alpha=&MakeFloatFromInt(Int=$Particles.PartitionParticles.Partition))
    Index 0 Param 2=&Lerp_Float(A=1, B=1.6, Alpha=&MakeFloatFromInt(Int=$Particles.PartitionParticles.Partition))
    Index 0 Param 3=&Lerp_Float(A=1, B=1.2, Alpha=&MakeFloatFromInt(Int=$Particles.PartitionParticles.Partition))
    Index 0 Param 4=&Lerp_Float(A=0, B=0.2, Alpha=&MakeFloatFromInt(Int=$Particles.PartitionParticles.Partition))
    Index 1 Param 1=&MakeFloatFromVector(Vector=$Particles.InterpolateOverTime003.InterpolateOverTime_V3_.Moving Average, Channel=::NewEnumerator0)
    Index 1 Param 2=&MakeFloatFromVector(Vector=$Particles.InterpolateOverTime003.InterpolateOverTime_V3_.Moving Average, Channel=::NewEnumerator1)
    Index 1 Param 3=&MakeFloatFromVector(Vector=$Particles.InterpolateOverTime003.InterpolateOverTime_V3_.Moving Average, Channel=::NewEnumerator2)
    Index 1 Param 4=0.5
    Index 2 Param 1=&Lerp_Float(
     A=&Lerp_Float(A=2.9, B=3.3, Alpha=&MakeFloatFromInt(Int=$Particles.PartitionParticles.Partition))
     B=3.7
     Alpha=$User.ForwardDot
    )
    Index 2 Param 2=&Lerp_Float(A=1.7, B=1, Alpha=&MakeFloatFromInt(Int=$Particles.PartitionParticles.Partition))
    Index 2 Param 3=&SelectFloatFromArray(Direct Array Index=$Particles.PartitionParticles.Partition, Float Selection Array=DataInterface() {propertyValues {FloatData [0,2]}})
    Index 2 Param 4=&SelectFloatFromArray(Direct Array Index=$Particles.PartitionParticles.Partition, Float Selection Array=DataInterface() {propertyValues {FloatData [0,4]}})
    Index 3 Param 1=4
    Index 3 Param 2=50
    Module.Param0WriteEnabled=true
    Module.Param0WriteEnabled1=true
    Module.Param0WriteEnabled2=true
    Module.Param0WriteEnabled3=true
    Module.Param1WriteEnabled=true
    Module.Param1WriteEnabled1=true
    Module.Param1WriteEnabled2=true
    Module.Param1WriteEnabled3=true
    Module.Param2WriteEnabled=true
    Module.Param2WriteEnabled1=true
    Module.Param2WriteEnabled2=true
    Module.Param2WriteEnabled3=true
    Module.Param3WriteEnabled=true
    Module.Param3WriteEnabled1=true
    Module.Param3WriteEnabled2=true
    Module.Param3WriteEnabled3=true
    Write Parameter Index 1=true
    Write Parameter Index 2=true
    Write Parameter Index 3=true
   ) {module=DynamicMaterialParameters001}
   &CameraOffset(Camera Offset Amount=2) {module=CameraOffset001}
   /SetVariables(Particles.VisibilityTag=$Particles.PartitionParticles.Partition) {isSetParameters=true, module=SetVariables_#11}
  ]
  rendererClasses [&NiagaraMeshRendererProperties]
  renderers [
   {rendererClass=&NiagaraMeshRendererProperties, rendererIndex=0}
   {rendererClass=&NiagaraMeshRendererProperties, rendererIndex=1}
  ]
 }
 Flame001 {
  emitterUpdateScript [
   &EmitterState(
    Inactive Response=::NewEnumerator1
    Life Cycle Mode=::NewEnumerator1
    Loop Behavior=::NewEnumerator1
    Loop Duration Mode=::NewEnumerator1
    UseLoopDelay=true
   )
   &SpawnBurst_Instantaneous(Spawn Count=2, Use Spawn Probability=false)
   &Waveform(
    EnableExponent1=false
    GlobalClampMax=false
    GlobalClampMin=false
    UseInput1=false
    UsePhase1=false
    "Waveform [1]":ENiagara_Waveforms=::NewEnumerator12
    "[1] Frequency"=7
   )
   /SetVariables(Emitter.Flicker:NiagaraFloat=&Add_Float(A=&Multiply_Float(A=$Emitter.Waveform.Waveform, B=0.5), B=0.5)) {isSetParameters=true, module=SetVariables_#6}
   /SetVariables(Emitter.bInterpolation=true) {isSetParameters=true, module=SetVariables_#7}
  ]
  particleSpawnScript [
   &InitializeParticle(
    Mass Mode=::NewEnumerator1
    Mesh Scale=&Multiply_Vector(A=&MakeVector(X=3.4, Y=2, Z=2), B=&VectorFromFloat(Value=0.6))
    Mesh Scale Mode=::NewEnumerator4
    Position Mode=::NewEnumerator0
    Sprite Size Mode=::NewEnumerator3
    Uniform Sprite Size=8
    Write Material Random=true
   ) {module=InitializeParticle001}
   &PartitionParticles()
   &CameraOffset()
   /SetVariables() {isSetParameters=true, module=SetVariables_#8}
   &SkeletalMeshLocation(
    Apply Sampled Orientation to Mesh Particles=true
    Bone / Socket Index=&SelectIntFromArray(Direct Array Index=$Particles.PartitionParticles.Partition, Int Selection Array=DataInterface() {propertyValues {IntData [0,1]}})
    Inherit Velocity=false
    Interpolate Mesh Sampling=false
    MaskSpawnGroup=false
    Mesh Sampling Type=::NewEnumerator5
    Sampled Position Offset Coordinate Space:ENiagaraMeshTransforms=::NewEnumerator3
    Skeletal Mesh=$System.Skeletal Mesh
    Socket Sampling Mode=::NewEnumerator5
    Use Sampled Mesh Attribute as Ribbon ID=false
   ) {module=SkeletalMeshLocation001}
   &FN_Sprite_Attributes(
    Color=&LerpLinearColors(
     StartColor=<rgba 1 0.9047 0.2831 1>
     EndColor=<rgba 0.9823 0.9131 0.2582 1>
     LerpFactor=&FN_FrontEndFloatModifier(Gameplay=0, Front End=1)
    )
    Create Color Attribute=true
    Create SpriteSize Attribute=true
    SpriteSize=<20 20>
    Write to Module Namespace=true
   ) {module=FN_Sprite_Attributes001}
   &InitialMeshOrientation(
    Mesh Axis To Orient=&SelectVectorFromArray(
     Direct Array Index=$Particles.PartitionParticles.Partition
     Vector Selection Array=DataInterface() {propertyValues {FloatData [<1 0 0>, <-1 0 0>]}}
    )
    Mesh Orientation Mode=::NewEnumerator2
    Orientation Vector=&RotateVectorByQuaternion(Vector To Rotate=<0 1 0>, Quaternion=$Output.SkeletalMeshLocation001.SampledRotation)
    Use Rotation Vector=false
   )
  ]
  particleUpdateScript [
   &ParticleState()
   &ScaleMeshSize(Scale Factor=&MakeVector(
     X=1.2
     Y=1.2
     Z=&FloatFromCurve(
      FloatCurve=DataInterface() {propertyValues {Curve {keys [
          {arriveTangent=9.0725, interpMode=RCIM_Cubic, leaveTangent=9.0725, tangentMode=RCTM_User, time=0, value=0}
          {arriveTangent=0.0017, interpMode=RCIM_Cubic, leaveTangent=0.0017, tangentMode=RCTM_User, time=0.1, value=0.9847}
          {arriveTangent=-0.0405, interpMode=RCIM_Cubic, leaveTangent=-0.0405, tangentMode=RCTM_User, time=1, value=0}
         ]}}}
      Scale Curve=2
     )
    ))
   &SkeletalMeshLocation(
    Apply Sampled Orientation to Mesh Particles=true
    Bone / Socket Index=&SelectIntFromArray(Direct Array Index=$Particles.PartitionParticles.Partition, Int Selection Array=DataInterface() {propertyValues {IntData [0,1]}})
    Inherit Velocity=false
    Interpolate Mesh Sampling=false
    MaskSpawnGroup=false
    Mesh Sampling Type=::NewEnumerator5
    Sampled Position Offset=&SelectVectorFromArray(
     Direct Array Index=$Particles.Initial.PartitionParticles.Partition
     Vector Selection Array=DataInterface() {propertyValues {FloatData [<-2.5 0 0>, <2.5 0 0>]}}
    )
    Sampled Position Offset Coordinate Space:ENiagaraMeshTransforms=::NewEnumerator3
    Skeletal Mesh=$System.Skeletal Mesh
    Socket Sampling Mode=::NewEnumerator5
    Use Sampled Mesh Attribute as Ribbon ID=false
   )
   /SetVariables(Particles.Velocity=&RotateVectorByQuaternion(Vector To Rotate=<0 0 1>, Quaternion=$Output.SkeletalMeshLocation.SampledRotation)) {isSetParameters=true, module=SetVariables_#9}
   &FN_Sprite_Attributes(
    Camera Offset=5
    Create Camera Offset Attribute=true
    Create Color Attribute=true
    Create Position Attribute=true
    Position:NiagaraPosition=&AddVectorToPosition(
     Position=$Output.SkeletalMeshLocation.SampledPosition
     Vector=&TransformVector(Vector=<-0.8 0 2.5>, Destination Space=::World)
    )
    Write to Module Namespace=true
   )
   &ScaleColor(Scale Alpha=&Multiply_Float(A=$Emitter.Flicker, B=4), ScaleA=true, ScaleRGB=true)
   &InterpolateOverTime(Rate Of Change=4, Target Float Value:NiagaraFloat=$User.ForwardDot) {module=InterpolateOverTime001}
   &ScaleMeshSize(Scale Factor=&MakeVector(
     X=&FloatFromCurve(
      FloatCurve=DataInterface() {propertyValues {Curve {keys [
          {interpMode=RCIM_Cubic, tangentMode=RCTM_User, time=0, value=0}
          {interpMode=RCIM_Cubic, tangentMode=RCTM_User, time=0.3, value=1}
          {tangentMode=RCTM_User, time=1, value=1}
         ]}}}
      Scale Curve=&Lerp_Float(A=1, B=1.25, Alpha=$Particles.InterpolateOverTime001.InterpolateOverTime_Float.Moving Average)
     )
     Y=0.5
     Z=&FloatFromCurve(
      FloatCurve=DataInterface() {propertyValues {Curve {keys [
          {arriveTangent=15.3111, interpMode=RCIM_Cubic, leaveTangent=15.3111, tangentMode=RCTM_User, time=0, value=0}
          {arriveTangent=-0.0274, interpMode=RCIM_Cubic, leaveTangent=-0.0274, tangentMode=RCTM_User, time=0.2, value=1.375}
          {tangentMode=RCTM_User, time=1, value=1}
         ]}}}
      Scale Curve=&Lerp_Float(A=0.8, B=1.5, Alpha=$Particles.InterpolateOverTime001.InterpolateOverTime_Float.Moving Average)
     )
    )) {module=ScaleMeshSize001}
   &ScaleColor(
    Scale Mode=::NewEnumerator1
    Scale RGBA=&MakeVector4_FromLinearColor(LinearColor=$System.TODColor)
    ScaleRGBA=true
   ) {module=ScaleColor001}
   &ScaleColor(
    Scale Alpha=&FloatFromCurve(
     FloatCurve=DataInterface() {propertyValues {Curve {keys [<at 0.2 1 RCIM_Cubic>,<at 0.4 1 RCIM_Cubic>]}}}
     CurveIndex=$System.ReactivityTiming
    )
    Scale RGB=&VectorFromFloat(Value=&FN_FrontEndFloatModifier(
      Gameplay=&FloatFromCurve(
       FloatCurve=DataInterface() {propertyValues {Curve {keys [<at 0.2 1 RCIM_Cubic>, {interpMode=RCIM_Cubic, tangentMode=RCTM_User, time=0.25, value=5}, <at 0.4 1 RCIM_Cubic>]}}}
       CurveIndex=$System.ReactivityTiming
      )
      Front End=&FloatFromCurve(
       FloatCurve=DataInterface() {propertyValues {Curve {keys [<at 0.2 1 RCIM_Cubic>, {interpMode=RCIM_Cubic, tangentMode=RCTM_User, time=0.25, value=3}, <at 0.4 1 RCIM_Cubic>]}}}
       CurveIndex=$System.ReactivityTiming
      )
     ))
    ScaleA=true
    ScaleRGB=false
   ) {module=ScaleColor002}
   /SetVariables(Particles.GlowColor=&MultiplyLinearColor(
     Linear Color=$Particles.Color
     Scale Factor=&MakeVector4_FromLinearColor(LinearColor=<rgba 1 0.8963 0.2747 1>)
    )) {isSetParameters=true, module=SetVariables_#10}
   &InterpolateOverTime(Rate Of Change=3, Target Float Value:NiagaraFloat=$User.RightDot) {module=InterpolateOverTime002}
   &InterpolateOverTime(Rate Of Change=20, Target Vector Value:NiagaraPosition=$Particles.Position, Variable Type=::NewEnumerator2) {module=InterpolateOverTime003}
   &DynamicMaterialParameters(
    Index 0 Param 1=&Lerp_Float(A=3.5, B=5, Alpha=&MakeFloatFromInt(Int=$Particles.PartitionParticles.Partition))
    Index 0 Param 2=&Lerp_Float(A=1, B=1.6, Alpha=&MakeFloatFromInt(Int=$Particles.PartitionParticles.Partition))
    Index 0 Param 3=&Lerp_Float(A=1, B=1.2, Alpha=&MakeFloatFromInt(Int=$Particles.PartitionParticles.Partition))
    Index 0 Param 4=&Lerp_Float(A=0, B=0.2, Alpha=&MakeFloatFromInt(Int=$Particles.PartitionParticles.Partition))
    Index 1 Param 1=&MakeFloatFromVector(Vector=$Particles.InterpolateOverTime003.InterpolateOverTime_V3_.Moving Average, Channel=::NewEnumerator0)
    Index 1 Param 2=&MakeFloatFromVector(Vector=$Particles.InterpolateOverTime003.InterpolateOverTime_V3_.Moving Average, Channel=::NewEnumerator1)
    Index 1 Param 3=&MakeFloatFromVector(Vector=$Particles.InterpolateOverTime003.InterpolateOverTime_V3_.Moving Average, Channel=::NewEnumerator2)
    Index 1 Param 4=0.5
    Index 2 Param 1=&Lerp_Float(
     A=&Lerp_Float(A=2.9, B=3.3, Alpha=&MakeFloatFromInt(Int=$Particles.PartitionParticles.Partition))
     B=3.7
     Alpha=$User.ForwardDot
    )
    Index 2 Param 2=&Lerp_Float(A=1.7, B=1, Alpha=&MakeFloatFromInt(Int=$Particles.PartitionParticles.Partition))
    Index 2 Param 3=&SelectFloatFromArray(Direct Array Index=$Particles.PartitionParticles.Partition, Float Selection Array=DataInterface() {propertyValues {FloatData [0,2]}})
    Index 2 Param 4=&SelectFloatFromArray(Direct Array Index=$Particles.PartitionParticles.Partition, Float Selection Array=DataInterface() {propertyValues {FloatData [0,4]}})
    Index 3 Param 1=4
    Index 3 Param 2=50
    Module.Param0WriteEnabled=true
    Module.Param0WriteEnabled1=true
    Module.Param0WriteEnabled2=true
    Module.Param0WriteEnabled3=true
    Module.Param1WriteEnabled=true
    Module.Param1WriteEnabled1=true
    Module.Param1WriteEnabled2=true
    Module.Param1WriteEnabled3=true
    Module.Param2WriteEnabled=true
    Module.Param2WriteEnabled1=true
    Module.Param2WriteEnabled2=true
    Module.Param2WriteEnabled3=true
    Module.Param3WriteEnabled=true
    Module.Param3WriteEnabled1=true
    Module.Param3WriteEnabled2=true
    Module.Param3WriteEnabled3=true
    Write Parameter Index 1=true
    Write Parameter Index 2=true
    Write Parameter Index 3=true
   ) {module=DynamicMaterialParameters001}
   &CameraOffset(Camera Offset Amount=2) {module=CameraOffset001}
   /SetVariables(Particles.VisibilityTag=$Particles.PartitionParticles.Partition) {isSetParameters=true, module=SetVariables_#11}
  ]
  rendererClasses [&NiagaraMeshRendererProperties]
  renderers [
   {rendererClass=&NiagaraMeshRendererProperties, rendererIndex=0}
   {rendererClass=&NiagaraMeshRendererProperties, rendererIndex=1}
  ]
 }
 Flare {
  emitterUpdateScript [
   &EmitterState(Life Cycle Mode=::NewEnumerator1, Loop Behavior=::NewEnumerator1, Loop Duration Mode=::NewEnumerator1, UseLoopDelay=true)
   &SpawnBurst_Instantaneous(Spawn Count=2, Use Spawn Probability=false, UseLoopCountLimit=false)
  ]
  particleSpawnScript [
   &FN_SetSpawnGroup(Number of Spawn Groups=2)
   &InitializeParticle(
    Color=<rgba 0.9889 0.6769 0.0515 1>
    Lifetime=0.25
    Position Offset=<0 0 50>
    Sprite Size=&Multiply_Vector2D_ByFloat(
     Float=&SelectFloatFromArray(Direct Array Index=$Particles.SpawnGroup, Float Selection Array=DataInterface() {propertyValues {FloatData [1.25,1.25,3]}})
     Vector2D=<20 30>
    )
    Sprite Size Mode=::NewEnumerator4
    UsePositionOffset=false
    Write Material Random=true
   )
   &PartitionParticles()
   &CameraOffset(Camera Offset Amount=8) {module=CameraOffset001}
  ]
  particleUpdateScript [
   &ParticleState(Kill Particles When Lifetime Has Elapsed=false, Let Infinitely Lived Particles Die When Emitter Deactivates=true)
   &ScaleSpriteSize(Scale Sprite Size Mode=::NewEnumerator0, Uniform Scale Factor=0.8) {module=ScaleSpriteSize001}
   &SkeletalMeshLocation(
    Apply Sampled Orientation to Mesh Particles=true
    Bone / Socket Index=&SelectIntFromArray(Direct Array Index=$Particles.PartitionParticles.Partition, Int Selection Array=DataInterface() {propertyValues {IntData [0,1]}})
    Inherit Velocity=false
    Interpolate Mesh Sampling=false
    MaskSpawnGroup=false
    Mesh Sampling Type=::NewEnumerator5
    Sampled Position Offset=&SelectVectorFromArray(
     Direct Array Index=$Particles.PartitionParticles.Partition
     Vector Selection Array=DataInterface() {propertyValues {FloatData [<0 0 0>, <0 0 0>]}}
    )
    Sampled Position Offset Coordinate Space:ENiagaraMeshTransforms=::NewEnumerator3
    Skeletal Mesh=$System.Skeletal Mesh
    Socket Sampling Mode=::NewEnumerator5
    Use Sampled Mesh Attribute as Ribbon ID=false
   ) {module=SkeletalMeshLocation001}
   &InterpolateOverTime(Rate Of Change=4, Target Float Value:NiagaraFloat=$User.ForwardDot) {module=InterpolateOverTime001}
   &ScaleColor(
    Scale Alpha=&Multiply_Float(
     A=&FloatFromCurve(FloatCurve=DataInterface() {propertyValues {Curve {keys [{interpMode=RCIM_Cubic, tangentMode=RCTM_User, time=0, value=0}, {interpMode=RCIM_Cubic, tangentMode=RCTM_User, time=1, value=1}]}}})
     B=1
    )
    Scale RGB=&Multiply_VectorByFloat(
     Vector=<1 1 1>
     Float=&Lerp_Float(A=2, B=3, Alpha=$Particles.InterpolateOverTime001.InterpolateOverTime_Float.Moving Average)
    )
    ScaleA=true
    ScaleRGB=true
   )
   &DynamicMaterialParameters(
    Index 0 Param 1=0.2
    Index 0 Param 2=0.02
    Module.Param0WriteEnabled=true
    Module.Param1WriteEnabled=true
    Module.Param2WriteEnabled=true
    Module.Param3WriteEnabled=true
   )
  ]
  rendererClasses [&NiagaraSpriteRendererProperties, &NiagaraLightRendererProperties]
  renderers [
   {rendererClass=&NiagaraSpriteRendererProperties}
   {rendererClass=&NiagaraLightRendererProperties, rendererIndex=1}
  ]
 }
 Flowers {
  emitterUpdateScript [
   &EmitterState(Delay First Loop Only=true, Life Cycle Mode=::NewEnumerator1, Loop Delay=0.2, Loop Duration=0.5, UseLoopDelay=true)
   &SpawnBurst_Instantaneous(
    Spawn Count=&RandomRangeInt(Minimum=6, Maximum=8)
    Use Spawn Probability=false
    UseLoopCountLimit=false
   )
  ]
  particleSpawnScript [
   &InitializeParticle(
    Color Mode=::NewEnumerator0
    Lifetime Max=1
    Lifetime Min=0.3
    Lifetime Mode=::NewEnumerator1
    Sprite Size Mode=::NewEnumerator3
    Uniform Sprite Size=&RandomRangeFloat(Minimum=7, Maximum=16)
    UsePositionOffset=false
    Write Material Random=true
   )
   &PartitionParticles()
   &ShapeLocation(
    MaskSpawnGroup=false
    Override Seed=false
    Ring / Disc Distribution Mode:ENiagaraSphereDistributionMode=::NewEnumerator1
    Ring Radius=2
    Rotation Axis=<0 1 0>
    Rotation Mode=::NewEnumerator1
    Shape Primitive=::NewEnumerator4
    U Position=&ReturnNormalizedExecIndex()
    UseOwnerScale=true
   )
   &ApplyInitialForces()
   &InitialMeshOrientation(
    Mesh Axis To Orient=<0 0 1>
    Mesh Orientation Mode=::NewEnumerator2
    Orientation Vector=$Particles.ShapeLocation.ShapeBitangent
    Rotation=<1 0 4>
    Use Rotation Vector=true
   )
   &AlignSpriteToMeshOrientation(
    Mesh Orientation Relative Sprite Alignment Vector=<1 0 0>
    Mesh Orientation Relative Sprite Facing Vector:Vector3f=<0 1 0>
    SetAlignment=true
    SetFacing=true
   )
   /SetVariables(Particles.pOffset:NiagaraPosition=$Particles.Position) {isSetParameters=true}
  ]
  particleUpdateScript [
   &ParticleState()
   &ScaleSpriteSize() {enabled=false, module=ScaleSpriteSize001}
   &InterpolateOverTime(Rate Of Change=4, Target Float Value:NiagaraFloat=$User.ForwardDot) {module=InterpolateOverTime001}
   &ScaleSpriteSize(
    Uniform Curve Scale=&Lerp_Float(A=1, B=1.5, Alpha=$Particles.InterpolateOverTime001.InterpolateOverTime_Float.Moving Average)
    Uniform Curve Sprite Scale=DataInterface() {propertyValues {Curve {keys [{tangentMode=RCTM_User, time=0, value=0.2}, <at 1 1>]}}}
   )
   &InterpolateOverTime(Rate Of Change=3, Target Float Value:NiagaraFloat=$User.RightDot) {module=InterpolateOverTime002}
   &ScaleMeshSize(Scale Factor=&MakeVector(
     X=&FloatFromCurve(FloatCurve=DataInterface() {propertyValues {Curve {keys [
          {interpMode=RCIM_Cubic, tangentMode=RCTM_User, time=0, value=0}
          {interpMode=RCIM_Cubic, tangentMode=RCTM_User, time=0.3, value=1}
          {tangentMode=RCTM_User, time=1, value=1}
         ]}}})
     Y=0.5
     Z=&FloatFromCurve(
      FloatCurve=DataInterface() {propertyValues {Curve {keys [
          {arriveTangent=15.3111, interpMode=RCIM_Cubic, leaveTangent=15.3111, tangentMode=RCTM_User, time=0, value=0}
          {arriveTangent=-0.0274, interpMode=RCIM_Cubic, leaveTangent=-0.0274, tangentMode=RCTM_User, time=0.2, value=1.375}
          {tangentMode=RCTM_User, time=1, value=1}
         ]}}}
      Scale Curve=&Lerp_Float(A=0.8, B=1.25, Alpha=$Particles.InterpolateOverTime001.InterpolateOverTime_Float.Moving Average)
     )
    )) {enabled=false, module=ScaleMeshSize001}
   &SkeletalMeshLocation(
    Apply Sampled Orientation to Mesh Particles=true
    Bone / Socket Index=&SelectIntFromArray(Direct Array Index=$Particles.PartitionParticles.Partition, Int Selection Array=DataInterface() {propertyValues {IntData [0,1]}})
    Inherit Velocity=false
    Interpolate Mesh Sampling=false
    MaskSpawnGroup=false
    Mesh Sampling Type=::NewEnumerator5
    Sampled Position Offset Coordinate Space:ENiagaraMeshTransforms=::NewEnumerator2
    Skeletal Mesh=$System.Skeletal Mesh
    Socket Sampling Mode=::NewEnumerator5
    Use Sampled Mesh Attribute as Ribbon ID=false
   )
   &DynamicMaterialParameters(
    Index 0 Param 4=&FloatFromCurve(FloatCurve=DataInterface() {propertyValues {Curve {keys [<at 0.2 0>,<at 1 1>]}}}, Scale Curve=5)
    Module.Index0Param0Write=false
    Module.Index0Param1Write=false
    Module.Index0Param2Write=false
    Module.Index0Param3Write=true
    Module.Index1Param0Write=true
    Module.Index1Param1Write=true
    Module.Index1Param2Write=true
    Module.Index1Param3Write=true
    Write Parameter Index 1=true
   )
  ]
  rendererClasses [&NiagaraSpriteRendererProperties]
  renderers [{rendererClass=&NiagaraSpriteRendererProperties}]
 }
 Glow {
  emitterUpdateScript [
   &EmitterState(Life Cycle Mode=::NewEnumerator1, Loop Behavior=::NewEnumerator1, Loop Duration Mode=::NewEnumerator1, UseLoopDelay=false)
   &SpawnBurst_Instantaneous(Spawn Count=2, Use Spawn Probability=false, UseLoopCountLimit=false)
  ]
  particleSpawnScript [
   &InitializeParticle(
    Color=<rgba 0.9889 0.649 0 1>
    Lifetime Max=1
    Lifetime Min=0.3
    Lifetime Mode=::NewEnumerator1
    Sprite Size Mode=::NewEnumerator3
    Uniform Sprite Size=7.5
    UsePositionOffset=false
    Write Material Random=true
   )
   &PartitionParticles()
  ]
  particleUpdateScript [
   &ParticleState(Kill Particles When Lifetime Has Elapsed=false, Let Infinitely Lived Particles Die When Emitter Deactivates=true)
   &SkeletalMeshLocation(
    Apply Sampled Orientation to Mesh Particles=true
    Bone / Socket Index=&SelectIntFromArray(Direct Array Index=$Particles.PartitionParticles.Partition, Int Selection Array=DataInterface() {propertyValues {IntData [0,1]}})
    Inherit Velocity=false
    Interpolate Mesh Sampling=false
    MaskSpawnGroup=false
    Mesh Sampling Type=::NewEnumerator5
    Sampled Position Offset=<0 -2.5 0>
    Sampled Position Offset Coordinate Space:ENiagaraMeshTransforms=::NewEnumerator2
    Skeletal Mesh=$System.Skeletal Mesh
    Socket Sampling Mode=::NewEnumerator5
    Use Sampled Mesh Attribute as Ribbon ID=false
   )
   &ScaleColor(
    Scale Alpha=&FloatFromCurve(FloatCurve=DataInterface() {propertyValues {Curve {keys [{interpMode=RCIM_Cubic, tangentMode=RCTM_User, time=0, value=0}, {interpMode=RCIM_Cubic, tangentMode=RCTM_User, time=0.5, value=1}]}}})
    Scale RGB=&VectorFromFloat(Value=2)
    ScaleA=true
    ScaleRGB=true
   )
   &DynamicMaterialParameters(
    Index 0 Param 1=1
    Index 0 Param 2=1
    Module.Index0Param0Write=true
    Module.Index0Param1Write=true
    Module.Index0Param2Write=true
    Module.Index0Param3Write=true
   ) {module=DynamicMaterialParameters001}
   &SpriteFacingAndAlignment(Sprite Facing=<0 1 0>, WriteAlignment=false, WriteFacing=true)
  ]
  rendererClasses [&NiagaraSpriteRendererProperties]
  renderers [{rendererClass=&NiagaraSpriteRendererProperties}]
 }
 MeshRipples002 {
  emitterUpdateScript [
   &EmitterState(Life Cycle Mode=::NewEnumerator1, Loop Delay=0.1, Loop Duration=0.67, UseLoopDelay=false)
   &SpawnBurst_Instantaneous(Spawn Count=2, Use Spawn Probability=false, UseLoopCountLimit=false)
   &SpawnBurst_Instantaneous(Spawn Count=2, Spawn Time=0.05, Use Spawn Probability=false, UseLoopCountLimit=false) {module=SpawnBurst_Instantaneous001}
   &SpawnBurst_Instantaneous(Spawn Count=2, Spawn Time=0.15, Use Spawn Probability=false, UseLoopCountLimit=false) {module=SpawnBurst_Instantaneous002}
  ]
  particleSpawnScript [
   &InitializeParticle(
    Lifetime=0.25
    Mesh Scale Mode=::NewEnumerator3
    Mesh Uniform Scale=0.13
    Position Offset=<5 0 -2.5>
    Sprite Rotation Mode=::NewEnumerator0
    Sprite Size Mode=::NewEnumerator3
    Uniform Sprite Size=&Multiply_Float(A=175, B=1)
    UsePositionOffset=false
    Write Material Random=true
   )
   &CameraOffset()
   &FN_SetSpawnGroup(Number of Spawn Groups=2)
   &InitialMeshOrientation(Mesh Orientation Mode=::NewEnumerator2, Orientation Vector=<0 0 1>, Use Rotation Vector=false) {module=InitialMeshOrientation001}
  ]
  particleUpdateScript [
   &ParticleState()
   &SkeletalMeshLocation(
    Bone / Socket Index=$Particles.SpawnGroup
    Bone Sampling Mode=::NewEnumerator4
    MaskSpawnGroup=false
    Sampled Position Offset=<0 3 -0.5>
    Skeletal Mesh=$System.Skeletal Mesh
   )
   &ScaleMeshSize(Scale Factor=&VectorFromFloat(Value=&Add_Float(
      A=1
      B=&FloatFromCurve(
       FloatCurve=DataInterface() {propertyValues {Curve {keys [
           {arriveTangent=-0.0415, arriveTangentWeight=0.2235, leaveTangentWeight=0.2235, time=0.05, value=0}
           {arriveTangentWeight=0.9219, interpMode=RCIM_Cubic, leaveTangent=1.5675, leaveTangentWeight=0.9219, time=1, value=0.85}
          ]}}}
       Scale Curve=&RandomRangeFloat(Minimum=0.5, Maximum=1)
      )
     )))
   &ScaleColor(
    Scale Alpha=&FloatFromCurve(
     FloatCurve=DataInterface() {propertyValues {Curve {keys [
         {arriveTangent=-1.1111, interpMode=RCIM_Cubic, time=0, value=0}
         <at 0.1 1 RCIM_Cubic>
         {arriveTangent=-0.9656, interpMode=RCIM_Cubic, leaveTangent=-0.9656, tangentMode=RCTM_User, time=1, value=0}
        ]}}}
     Scale Curve=1.5
    )
    Scale RGB=&Multiply_VectorByFloat(
     Vector=&MakeVectorFromLinearColorRGB(Color=&ColorFromCurve(ColorCurve=DataInterface() {propertyValues {
         AlphaCurve {keys []}
         BlueCurve {keys [
           {interpMode=RCIM_Cubic, leaveTangentWeight=0.8633, time=0, value=1.25}
           {arriveTangentWeight=0.5327, interpMode=RCIM_Cubic, time=1, value=1.25}
          ]}
         GreenCurve {keys [
           {interpMode=RCIM_Cubic, leaveTangentWeight=0.8633, time=0, value=0.561}
           {
            arriveTangent=6.199
            arriveTangentWeight=0.5327
            interpMode=RCIM_Cubic
            leaveTangent=6.199
            leaveTangentWeight=0.6868
            tangentMode=RCTM_User
            time=1
            value=0.7871
           }
          ]}
         RedCurve {keys [
           {interpMode=RCIM_Cubic, leaveTangentWeight=0.8633, time=0, value=0}
           {
            arriveTangent=6.199
            arriveTangentWeight=0.5327
            interpMode=RCIM_Cubic
            leaveTangent=6.199
            leaveTangentWeight=0.6899
            tangentMode=RCTM_User
            time=1
            value=0.4102
           }
          ]}
        }}))
     Float=2
    )
    ScaleA=true
    ScaleRGB=true
   )
   &DynamicMaterialParameters(
    Index 0 Param 4=&FloatFromCurve(
     FloatCurve=DataInterface() {propertyValues {Curve {keys [
         {
          arriveTangent=9.177
          arriveTangentWeight=0.6327
          interpMode=RCIM_Cubic
          leaveTangent=9.177
          leaveTangentWeight=0.6327
          tangentMode=RCTM_User
          tangentWeightMode=RCTWM_WeightedBoth
          time=0
          value=0.2
         }
         {
          arriveTangentWeight=0.6785
          interpMode=RCIM_Cubic
          leaveTangentWeight=0.6785
          tangentMode=RCTM_User
          tangentWeightMode=RCTWM_WeightedBoth
          time=1
          value=1
         }
        ]}}}
     Scale Curve=0.125
    )
    Module.Index0Param0Write=true
    Module.Index0Param1Write=true
    Module.Index0Param2Write=true
    Module.Index0Param3Write=true
   )
   /SetVariables(Particles.VibratePosition:NiagaraPosition=&AddVectorToPosition(
     Position=$Particles.Position
     Vector=&Add_Vector(
      A=&Multiply_Vector(
       A=&Add_Vector(
        A=&Multiply_Vector(
         A=/Game/Effects/Niagara/DynamicInputs/Engine/FN_GetSystemAxis(Axis:ENiagaraOrientationAxis=::XAxis, Flip Axis:NiagaraBool=false)
         B=&VectorFromFloat(Value=&RandomRangeFloat(Minimum=-1, Maximum=-1, Evaluation Type=::NewEnumerator1))
        )
        B=&Multiply_Vector(
         A=&ReturnMeshOrientationAxis(Mesh Orientation Axis=::YAxis, Mesh Orientation Quat=$Particles.MeshOrientation)
         B=&VectorFromFloat(Value=&RandomRangeFloat(Minimum=-1, Maximum=1, Evaluation Type=::NewEnumerator1))
        )
       )
       B=&VectorFromFloat(Value=1)
      )
      B=&RotateVectorByQuaternion(
       Vector To Rotate=&MakeVector(
        X=&FloatFromCurve(
         FloatCurve=DataInterface() {propertyValues {Curve {keys [{tangentMode=RCTM_User, time=0, value=0}, {tangentMode=RCTM_User, time=1, value=3.5}]}}}
         Scale Curve=2
        )
        Y=0
        Z=0
       )
       Quaternion=$Particles.SkeletalMeshLocation.SampledRotation
      )
     )
    )) {isSetParameters=true}
  ]
  rendererClasses [&NiagaraMeshRendererProperties]
  renderers [
   {rendererClass=&NiagaraMeshRendererProperties, rendererIndex=0}
   {rendererClass=&NiagaraMeshRendererProperties, rendererIndex=1}
  ]
 }
 Speed_Lines {
  emitterUpdateScript [
   &EmitterState()
   &SpawnRate(
    SpawnRate=&FN_FrontEndFloatModifier(Gameplay=&Multiply_Float(A=$System.EffectIntensity, B=100), Front End=80)
    Use Spawn Probability=false
   ) {module=SpawnRate001}
  ]
  particleSpawnScript [
   &InitializeParticle(
    Color=&MakeLinearColorFromVectorAndFloat("Vector (RGB)"=<1 1 1>, "Float (Alpha)"=&FN_FrontEndFloatModifier(Gameplay=0.25, Front End=0.1))
    Lifetime=&FN_FrontEndFloatModifier(Gameplay=&RandomRangeFloat(Minimum=0.1, Maximum=0.16), Front End=0.5)
    Sprite Size Max=<4 10>
    Sprite Size Min=<3 5>
    Sprite Size Mode=::NewEnumerator2
    UsePositionOffset=true
    Write Material Random=true
   )
   &ShapeLocation(
    Handle Radius=5
    Large Radius=&FN_FrontEndFloatModifier(Gameplay=250, Front End=50)
    MaskSpawnGroup=false
    Offset=&FN_FrontEndVector3Modifier(Gameplay=<0 -200 15>, Front End=<0 200 -25>)
    Offset Mode=::NewEnumerator0
    Override Seed=false
    Rotation Axis=<0 1 0>
    Rotation Mode=::NewEnumerator1
    Shape Primitive=::NewEnumerator3
    UseOwnerScale=false
   )
   &AddVelocityFromPoint(
    Module.Use Falloff Distance=false
    Origin Offset=&FN_FrontEndVector3Modifier(Gameplay=<0 -1000 -250>, Front End=<0 -500 0>)
    Velocity Strength=&FN_FrontEndFloatModifier(Gameplay=1000, Front End=800)
   )
   &AddVelocity(
    Velocity=&FN_FrontEndVector3Modifier(
     Gameplay=&Add_Vector(A=<0 300 0>, B=&Multiply_VectorByFloat(Vector=<0 0 222>, Float=$System.EffectIntensity))
     Front End=<0 -300 8>
    )
    Velocity Speed Scale=&FN_FrontEndFloatModifier(Gameplay=-1, Front End=4)
   )
  ]
  particleUpdateScript [
   &ParticleState()
   &ScaleSpriteSize(Scale Factor=&Vector2DFromCurve(
     Vector2Curve=DataInterface() {propertyValues {XCurve {keys [<at 0 0.5 RCIM_Cubic>,<at 0.5 1 RCIM_Cubic>]}, YCurve {keys [<at 0 0 RCIM_Cubic>,<at 0.5 1 RCIM_Cubic>]}}}
     ScaleCurve=&Vector2DFromFloat(Value=1)
    ))
   &SolveForcesAndVelocity(Clamp Velocity=false, Limit Acceleration=false)
   &ScaleSpriteSizeBySpeed(
    Max Scale Factor=&FN_FrontEndVector2Modifier(Front End:Vector2f=<1 25>, Gameplay:Vector2f=<1 50>)
    Min Scale Factor=<0.01 0.05>
    Velocity Threshold=1
   )
   &ScaleColor(
    Scale Alpha=&FloatFromCurve(
     FloatCurve=DataInterface() {propertyValues {Curve {keys [
         {arriveTangent=6.8397, interpMode=RCIM_Cubic, leaveTangent=6.8397, tangentMode=RCTM_User, time=0, value=0}
         <at 0.3 1 RCIM_Cubic>
         {arriveTangent=-2.221, interpMode=RCIM_Cubic, leaveTangent=-2.221, tangentMode=RCTM_User, time=1, value=0}
        ]}}}
     Scale Curve=&FN_FrontEndFloatModifier(
      Gameplay=&Lerp_Float(
       A=0
       B=&FNTimeOfDayScalar(Night Scale=0.4, Morning Scale=0.5, Day Scale=0.8, Evening Scale=0.4)
       Alpha=&Multiply_Float(A=$System.FadeIn, B=$System.Timeline001.Time)
      )
      Front End=&Lerp_Float(A=0, B=0.1, Alpha=$System.FadeIn)
     )
    )
    Scale RGB=&VectorFromFloat(Value=&FN_FrontEndFloatModifier(Gameplay=0.5, Front End=1))
    ScaleA=true
    ScaleRGB=true
   )
  ]
  rendererClasses [&NiagaraSpriteRendererProperties]
  renderers [{rendererClass=&NiagaraSpriteRendererProperties}]
 }
}
summary {
 systemName=NS_Glider_LemonCartGranite_Trail_Simple
 emitters [
  {emitterName=Flame, rendererClasses [&NiagaraMeshRendererProperties]}
  {emitterName=Speed_Lines, rendererClasses [&NiagaraSpriteRendererProperties]}
  {emitterName=Flare, rendererClasses [&NiagaraSpriteRendererProperties, &NiagaraLightRendererProperties]}
  {emitterName=Flowers, rendererClasses [&NiagaraSpriteRendererProperties]}
  {emitterName=MeshRipples002, rendererClasses [&NiagaraMeshRendererProperties]}
  {emitterName=Glow, rendererClasses [&NiagaraSpriteRendererProperties]}
  {emitterName=Flame001, rendererClasses [&NiagaraMeshRendererProperties]}
 ]
 userVariables [
  User.bIsFrontEnd {type=NiagaraBool}
  User.bIsFrontEndPreview {type=NiagaraBool}
  User.bIsFullyDeployed {type=NiagaraBool}
  User.CharacterPart {type=NiagaraFloat, defaultValue=0, description=""}
  User.EngineDelay {type=NiagaraFloat, defaultValue=0.16, description=""}
  User.FE Intergrator Value {type=NiagaraFloat, defaultValue=0.5, description=""}
  User.ForwardDot {type=NiagaraFloat, defaultValue=0, description=""}
  User.FrontEndScale {type=NiagaraFloat, defaultValue=0.5, description=""}
  User.Interpolate Intensity {type=NiagaraFloat, defaultValue=0.5, description=""}
  User.RightDot {type=NiagaraFloat, defaultValue=0, description=""}
  User.Spawn Delay {type=NiagaraFloat, defaultValue=1.5, description=""}
 ]
}
