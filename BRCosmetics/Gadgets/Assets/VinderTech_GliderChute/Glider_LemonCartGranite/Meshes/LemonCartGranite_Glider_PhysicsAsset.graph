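# Physics asset for the LemonCartGranite glider, a pedal trike cart with thrusters, cooler and boombox.
# Each simulated bone gets one capsule body, including wheels, covers and three sampled chain bones.
# Constraints form a tree from dyn_bike_MSTR: a pedal branch carries cranks, seat, chain, rear wheel and thrust/cargo assembly; a handle branch carries fork, front wheels, tubes and handlebars.
# All constraints share identical swing and twist limit angles.
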PhysicsAsset /BRCosmetics/Gadgets/Assets/VinderTech_GliderChute/Glider_LemonCartGranite/Meshes/LemonCartGranite_Glider_PhysicsAsset
bodies {
 bike_chain_19 [{shapeType=Capsule, length=9.361, radius=8.0799, center=<10.6462 -0.1912 5.6226>, rotation=<rot 0 -89.9008 -80.9538>}]
 bike_chain_3 [{shapeType=Capsule, length=20.911, radius=2.0196, center=<18.1971 0.191 -11.4643>, rotation=<rot 0 -89.9962 -51.0385>}]
 bike_chain_34 [{shapeType=Capsule, length=23.2653, radius=3.713, center=<25.6134 -0.2075 2.5421>, rotation=<rot 0 89.9683 -81.6987>}]
 dyn_bike_MSTR [{shapeType=Capsule, length=77.817, radius=34.0802, center=<-3.1056 -0.0003 -17.9615>, rotation=<rot 0 90.0003 -67.0442>}]
 dyn_boombox [{shapeType=Capsule, length=23.9203, radius=14.3934, center=<3.8928 -11.0154 -0.1958>, rotation=<rot -0.9508 -90 0.0901>}]
 dyn_cooler [{shapeType=Capsule, length=29.5861, radius=17.0213, center=<11.2292 0.5712 0.0352>, rotation=<rot 0.1238 -90 0.1559>}]
 dyn_cooler_bar [{shapeType=Capsule, length=20.0039, radius=16.4997, center=<14.6563 17.7265 -0.0393>, rotation=<rot 0 1.2415 -89.9865>}]
 dyn_fork [{shapeType=Capsule, length=39.6324, radius=9.5916, center=<0.8427 0 0.16>, rotation=<rot 0 -179.9999 -90>}]
 dyn_handle [{shapeType=Capsule, length=28.2129, radius=9.1823, center=<13.5618 0 -4.1417>, rotation=<rot 0 90 -80.6819>}]
 dyn_handlebar_MSTR [{shapeType=Capsule, length=29.3073, radius=14.3788, center=<-5.7868 -0.2147 12.3622>, rotation=<rot 0 -0.3975 -89.1815>}]
 dyn_handlebar_l_02 [{shapeType=Capsule, length=13.9558, radius=12.3341, center=<5.5027 -0.4343 9.7429>, rotation=<rot 0 -74.4214 -71.2028>}]
 dyn_handlebar_r_02 [{shapeType=Capsule, length=13.816, radius=12.7144, center=<-5.35 0.4687 -9.9203>, rotation=<rot 0 -74.3561 -68.7034>}]
 dyn_lid [{shapeType=Capsule, length=27.9807, radius=17.5291, center=<16.438 3.349 0.0081>, rotation=<rot -0.1456 -90 -0.5487>}]
 dyn_pedal_MSTR [{shapeType=Capsule, length=45.0362, radius=29.3475, center=<-12.1013 -3.9744 17.6342>, rotation=<rot -5.8272 -90 30.8938>}]
 dyn_pedal_bracket_r_fr [{shapeType=Capsule, length=10.5448, radius=10.544, center=<1.8151 -0.0008 0.1601>, rotation=<rot 11.8678 -90 0.1241>}]
 dyn_pedal_crank_l_02 [{shapeType=Capsule, length=15.7414, radius=8.6348, center=<15.7558 -0.23 -7.8162>, rotation=<rot -1.8931 -90 -39.9541>}]
 dyn_pedal_crank_r_02 [{shapeType=Capsule, length=15.471, radius=10.5578, center=<13.1721 -0.163 7.8129>, rotation=<rot 0 91.5556 -55.3345>}]
 dyn_seat [{shapeType=Capsule, length=26.3142, radius=9.2338, center=<-0.6203 -1.627 -0.0075>, rotation=<rot 0 92.2411 -90>}]
 dyn_thrust_MSTR [{shapeType=Capsule, length=20.8525, radius=19.8087, center=<-11.079 2.1475 0.6036>, rotation=<rot 14.2086 167.2833 -90>}]
 dyn_thruster_l_03 [{shapeType=Capsule, length=18.0193, radius=4.9479, center=<17.629 -0.5585 0.7261>, rotation=<rot 0 90.2244 -84.0445>}]
 dyn_thruster_r_03 [{shapeType=Capsule, length=18.8206, radius=4.9479, center=<-16.899 0.5698 -1.0015>, rotation=<rot 0 90.2509 -85.1307>}]
 dyn_tube_top_l [{shapeType=Capsule, length=21.817, radius=1.818, center=<21.3455 0.1435 -0.1019>, rotation=<rot 0 90.1616 -89.657>}]
 dyn_tube_top_r [{shapeType=Capsule, length=21.817, radius=1.818, center=<21.3455 0.1435 0.1019>, rotation=<rot 0 -89.8384 -89.657>}]
 dyn_wheel_bk [{shapeType=Capsule, length=23.5637, radius=23.5637, center=<0 -0.0001 -0.0001>, rotation=<rot 0 -90 0.4151>}]
 dyn_wheel_cover_l [{shapeType=Capsule, length=26.7314, radius=15.3842, center=<0.0015 -6.8225 9.994>, rotation=<rot 0 -180 -62.1513>}]
 dyn_wheel_cover_r [{shapeType=Capsule, length=26.7314, radius=15.3842, center=<0.0054 6.8225 -9.994>, rotation=<rot 0 -180 -62.1513>}]
 dyn_wheel_l_fr [{shapeType=Capsule, length=23.5636, radius=23.5636, center=<0.0001 -1.9431 -0.0004>, rotation=<rot -0.0098 -90 0.4324>}]
 dyn_wheel_r_fr [{shapeType=Capsule, length=23.5636, radius=23.5636, center=<0.0001 1.9361 -0.0004>, rotation=<rot 0.0098 -90 0.4324>}]
 dyn_wheelcover_bk [{shapeType=Capsule, length=26.7314, radius=15.3986, center=<7.6866 9.3255 0.1747>, rotation=<rot 0 44.0173 -89.6821>}]
}
constraints [
 {bone1Name=dyn_pedal_MSTR, bone2Name=dyn_bike_MSTR, swing1LimitDegrees=45, swing2LimitDegrees=45, twistLimitDegrees=45}
 {bone1Name=dyn_pedal_crank_l_02, bone2Name=dyn_pedal_MSTR, swing1LimitDegrees=45, swing2LimitDegrees=45, twistLimitDegrees=45}
 {bone1Name=dyn_pedal_bracket_r_fr, bone2Name=dyn_pedal_MSTR, swing1LimitDegrees=45, swing2LimitDegrees=45, twistLimitDegrees=45}
 {bone1Name=dyn_pedal_crank_r_02, bone2Name=dyn_pedal_bracket_r_fr, swing1LimitDegrees=45, swing2LimitDegrees=45, twistLimitDegrees=45}
 {bone1Name=dyn_thrust_MSTR, bone2Name=dyn_pedal_MSTR, swing1LimitDegrees=45, swing2LimitDegrees=45, twistLimitDegrees=45}
 {bone1Name=dyn_thruster_l_03, bone2Name=dyn_thrust_MSTR, swing1LimitDegrees=45, swing2LimitDegrees=45, twistLimitDegrees=45}
 {bone1Name=dyn_thruster_r_03, bone2Name=dyn_thrust_MSTR, swing1LimitDegrees=45, swing2LimitDegrees=45, twistLimitDegrees=45}
 {bone1Name=dyn_cooler_bar, bone2Name=dyn_thrust_MSTR, swing1LimitDegrees=45, swing2LimitDegrees=45, twistLimitDegrees=45}
 {bone1Name=dyn_cooler, bone2Name=dyn_cooler_bar, swing1LimitDegrees=45, swing2LimitDegrees=45, twistLimitDegrees=45}
 {bone1Name=dyn_boombox, bone2Name=dyn_cooler, swing1LimitDegrees=45, swing2LimitDegrees=45, twistLimitDegrees=45}
 {bone1Name=dyn_lid, bone2Name=dyn_cooler, swing1LimitDegrees=45, swing2LimitDegrees=45, twistLimitDegrees=45}
 {bone1Name=dyn_wheelcover_bk, bone2Name=dyn_thrust_MSTR, swing1LimitDegrees=45, swing2LimitDegrees=45, twistLimitDegrees=45}
 {bone1Name=dyn_wheel_bk, bone2Name=dyn_pedal_MSTR, swing1LimitDegrees=45, swing2LimitDegrees=45, twistLimitDegrees=45}
 {bone1Name=bike_chain_3, bone2Name=dyn_pedal_MSTR, swing1LimitDegrees=45, swing2LimitDegrees=45, twistLimitDegrees=45}
 {bone1Name=bike_chain_19, bone2Name=bike_chain_3, swing1LimitDegrees=45, swing2LimitDegrees=45, twistLimitDegrees=45}
 {bone1Name=bike_chain_34, bone2Name=bike_chain_19, swing1LimitDegrees=45, swing2LimitDegrees=45, twistLimitDegrees=45}
 {bone1Name=dyn_seat, bone2Name=dyn_pedal_MSTR, swing1LimitDegrees=45, swing2LimitDegrees=45, twistLimitDegrees=45}
 {bone1Name=dyn_handle, bone2Name=dyn_bike_MSTR, swing1LimitDegrees=45, swing2LimitDegrees=45, twistLimitDegrees=45}
 {bone1Name=dyn_fork, bone2Name=dyn_handle, swing1LimitDegrees=45, swing2LimitDegrees=45, twistLimitDegrees=45}
 {bone1Name=dyn_wheel_r_fr, bone2Name=dyn_fork, swing1LimitDegrees=45, swing2LimitDegrees=45, twistLimitDegrees=45}
 {bone1Name=dyn_wheel_cover_r, bone2Name=dyn_fork, swing1LimitDegrees=45, swing2LimitDegrees=45, twistLimitDegrees=45}
 {bone1Name=dyn_wheel_l_fr, bone2Name=dyn_fork, swing1LimitDegrees=45, swing2LimitDegrees=45, twistLimitDegrees=45}
 {bone1Name=dyn_wheel_cover_l, bone2Name=dyn_fork, swing1LimitDegrees=45, swing2LimitDegrees=45, twistLimitDegrees=45}
 {bone1Name=dyn_tube_top_l, bone2Name=dyn_handle, swing1LimitDegrees=45, swing2LimitDegrees=45, twistLimitDegrees=45}
 {bone1Name=dyn_tube_top_r, bone2Name=dyn_handle, swing1LimitDegrees=45, swing2LimitDegrees=45, twistLimitDegrees=45}
 {bone1Name=dyn_handlebar_MSTR, bone2Name=dyn_handle, swing1LimitDegrees=45, swing2LimitDegrees=45, twistLimitDegrees=45}
 {bone1Name=dyn_handlebar_l_02, bone2Name=dyn_handlebar_MSTR, swing1LimitDegrees=45, swing2LimitDegrees=45, twistLimitDegrees=45}
 {bone1Name=dyn_handlebar_r_02, bone2Name=dyn_handlebar_MSTR, swing1LimitDegrees=45, swing2LimitDegrees=45, twistLimitDegrees=45}
]
